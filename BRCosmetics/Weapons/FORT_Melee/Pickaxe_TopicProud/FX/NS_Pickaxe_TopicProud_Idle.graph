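# Niagara system for the TopicProud pickaxe's idle effect: one attached ribbon emitter sampling the weapon's skeletal mesh.
# Ribbons spawn continuously, drift upward in world space and follow source movement less as they age, while narrowing, eroding and fading.
# Alpha is zero while User.bisSwinging is set, otherwise User.ReactivityIndex gates it, showing only at higher levels.
# Particle lifetime still reads the FN_Integrator value, although that module is disabled.
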NiagaraSystem /BRCosmetics/Weapons/FORT_Melee/Pickaxe_TopicProud/FX/NS_Pickaxe_TopicProud_Idle
dependencies {
 usedDataInterfaces [&NiagaraDataInterfaceCurve, &NiagaraDataInterfaceColorCurve, &NiagaraDataInterfaceArrayFloat]
 usedDynamicInputs []
 usedModules [
  &SystemState
  &self:SystemSpawnScript.SystemScriptSource.SystemScriptGraph.NiagaraNodeAssignment_1.SetVariables_#1
  &Timeline
 ]
}
emitters {FN_AttachedRibbon {
  emitterUpdateScript [
   &EmitterState()
   &FN_SpawnRibbon_v2(SpawnRate=30)
   &FN_Integrator() {enabled=false}
  ]
  particleSpawnScript [
   /Game/Effects/Niagara/Modules/Ribbons/FN_InitializeRibbon_v4(
    Enable Skeletal Position=true
    Offset Amount=-15
    Offset Axis:ENiagaraOrientationAxis=::ZAxis
    Ribbon Width=33
    SkeletalMesh:NiagaraDataInterfaceSkeletalMesh=$User.SkeletalMesh
   )
   /SetVariables(Particles.RandomFromRibbonID:NiagaraFloat=/Game/Effects/Niagara/DynamicInputs/Random/FN_SeededRandomRangeFloat(Minimum:NiagaraFloat=0, Maximum:NiagaraFloat=1, Random Seed:NiagaraInt32=$Particles.SpawnGroup)) {isSetParameters=true}
   &AddVelocity(Coordinate Space=::World, Velocity=<0 0 95>)
   &ApplyInitialForces()
   &SkeletalMeshLocation(
    Apply To Particle Position=false
    Inherit Velocity=false
    MaskSpawnGroup=true
    Mesh Sampling Type=::NewEnumerator5
    Sampled Position Offset Coordinate Space:ENiagaraMeshTransforms=::NewEnumerator3
    Skeletal Mesh=$System.Skeletal Mesh
    Socket Sampling Mode=::NewEnumerator5
   )
  ]
  particleUpdateScript [
   /SetVariables(Particles.Lifetime=&FloatFromCurve(FloatCurve=DataInterface() {propertyValues {Curve {keys [<at 0.5 1>,<at 1 0.2>]}}}, CurveIndex=$System.FN_Integrator.Value, Scale Curve=0.5)) {isSetParameters=true}
   &SkeletalMeshLocation(
    Apply To Particle Position=false
    Inherit Velocity=false
    MaskSpawnGroup=false
    Mesh Sampling Type=::NewEnumerator5
    Skeletal Mesh=$System.Skeletal Mesh
    Socket Sampling Mode=::NewEnumerator5
   ) {enabled=false, module=SkeletalMeshLocation001}
   /SetVariables(Particles.DistanceFromSource:NiagaraFloat=&ClampFloat(Float=&Divide_Float(
      A=&Add_Float(
       A=0
       B=&VectorLength(Vector=&Subtract_Vector(A=Linked() {linkedVariable {name=Particles.Position, type=NiagaraPosition}}, B=$Output.SkeletalMeshLocation001.SampledPosition))
      )
      B=100
     ))) {enabled=false, isSetParameters=true}
   &InheritSourceMovement(
    Applied Position Scale=&VectorFromFloat(Value=&FloatFromCurve(FloatCurve=DataInterface() {propertyValues {Curve {keys [
          {arriveTangent=0.031, interpMode=RCIM_Cubic, leaveTangent=0.031, tangentMode=RCTM_User, time=0, value=0.8}
          {arriveTangent=0.2077, interpMode=RCIM_Cubic, leaveTangent=0.2077, tangentMode=RCTM_User, time=0.958, value=0.0615}
         ]}}}))
    ApplyForce=false
    ApplyPosition=true
    ApplyPositionFalloff=false
   )
   &SolveForcesAndVelocity(Clamp Velocity=false, Limit Acceleration=false)
   /Game/Effects/Niagara/Modules/Ribbons/FN_UpdateRibbon_v4(Enable Skeletal Position=true, SkeletalMesh:NiagaraDataInterfaceSkeletalMesh=$User.SkeletalMesh)
   &FN_RibbonMaterial(
    Color=&ColorFromCurve(ColorCurve=DataInterface() {propertyValues {
       AlphaCurve {keys [
         {arriveTangent=0.3686, interpMode=RCIM_Cubic, leaveTangent=0.3686, tangentMode=RCTM_User, time=0, value=0.0025}
         <at 0.2 0.0375>
         <at 1 0 RCIM_Cubic>
        ]}
       BlueCurve {keys [<at 0 1 RCIM_Cubic>]}
       GreenCurve {keys [<at 0 1 RCIM_Cubic>]}
       RedCurve {keys [<at 0 1 RCIM_Cubic>]}
      }})
    Erosion Amount:NiagaraFloat=&FloatFromCurve(
     FloatCurve=DataInterface() {propertyValues {Curve {keys [{arriveTangent=5.1387, interpMode=RCIM_Cubic, leaveTangent=5.1387, tangentMode=RCTM_User, time=0, value=0}, <at 1 1.5 RCIM_Cubic>]}}}
     Scale Curve=0.5
    )
    Erosion Pan Speed U=1
    Opacity Pan U Offset:NiagaraFloat=$Particles.RandomFromRibbonID
    Opacity Pan U Speed=0.5
    Scale Alpha:NiagaraFloat=&MakeCustomFloatFromBool(
     Bool=$User.bisSwinging
     True Float=0
     False Float=&SelectFloatFromArray(Direct Array Index=$User.ReactivityIndex, Float Selection Array=DataInterface() {propertyValues {FloatData [0,0,1,1]}})
    )
   ) {module=FN_RibbonMaterial001}
   &ScaleRibbonWidth(Ribbon Width Scale=&FloatFromCurve(FloatCurve=DataInterface() {propertyValues {Curve {keys [
         {arriveTangent=4.1493, interpMode=RCIM_Cubic, leaveTangent=4.1493, tangentMode=RCTM_User, time=0, value=0.7}
         {arriveTangent=-0.7, interpMode=RCIM_Cubic, leaveTangent=-0.7, time=0.1, value=1}
         {interpMode=RCIM_Cubic, tangentMode=RCTM_User, time=1, value=0}
        ]}}}))
   &KillParticles(Kill Particles=&SetBoolByFloatComparison(
     A=&MakeFloatFromLinearColor(LinearColor=$Particles.Color, Channel:ENiagaraLinearColor_Channels=::NewEnumerator3)
     B=0
     Comparison Type=::NewEnumerator4
    ))
  ]
  rendererClasses [&NiagaraRibbonRendererProperties]
  renderers [{rendererClass=&NiagaraRibbonRendererProperties}]
 }}
summary {
 systemName=NS_Pickaxe_TopicProud_Idle
 emitters [{emitterName=FN_AttachedRibbon, rendererClasses [&NiagaraRibbonRendererProperties]}]
 userVariables [
  User.bIsFrontEnd {type=NiagaraBool}
  User.bIsFrontendPreview {type=NiagaraBool}
  User.bisSwinging {type=NiagaraBool}
  User.IsSkydiving {type=NiagaraInt32}
  User.ReactivityIndex {type=NiagaraInt32}
  User.RibbonIDOffset {type=NiagaraInt32}
  User.SkeletalMesh {
   type=NiagaraDataInterfaceSkeletalMesh
   defaultValue {
    struct=NiagaraExt_VariableValue_DataInterface
    value {dataInterface=&self:NiagaraDataInterfaceSkeletalMesh_0}
   }
  }
 ]
}
