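# Niagara system for the TopicProud pickaxe's swing trail, with one anim-trail emitter that spawns short-lived particles from a bone or socket pair and draws three ribbon renderers.
# User.ReactivityIndex picks one of four color sets, each tinted by time of day.
# Curve-driven tints scale particle color, with alpha following System.SystemFade.
# A separate SoftTrail color blends per-index night, morning, day and evening colors through FNTimeOfDayColor.
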NiagaraSystem /BRCosmetics/Weapons/FORT_Melee/Pickaxe_TopicProud/FX/NS_Pickaxe_TopicProud_Trail
dependencies {
 usedDataInterfaces [
  &NiagaraDataInterfaceCurve
  &NiagaraDataInterfaceColorCurve
  &NiagaraDataInterfaceArrayColor
  &NiagaraDataInterfaceSkeletalMesh
  &NiagaraDataInterfaceSocketReader
 ]
 usedDynamicInputs []
 usedModules [
  &SystemState
  &self:SystemSpawnScript.SystemScriptSource.SystemScriptGraph.NiagaraNodeAssignment_0.SetVariables_#1
  &self:SystemSpawnScript.SystemScriptSource.SystemScriptGraph.NiagaraNodeAssignment_1.SetVariables_#2
  &self:SystemSpawnScript.SystemScriptSource.SystemScriptGraph.NiagaraNodeAssignment_2.SetVariables_#3
 ]
}
emitters {FN_AnimTrail {
  emitterSpawnScript [&FN_RibbonIDOffsetUtility()]
  emitterUpdateScript [
   &EmitterState(Life Cycle Mode=::NewEnumerator1, Loop Behavior=::NewEnumerator1, Loop Duration=1, UseLoopDelay=false)
   &FN_SpawnAnimTrail()
  ]
  particleSpawnScript [
   &FN_InitializeAnimTrailBoneIndecies(First Bone or Socket Index=1, Second Bone or Socket Index=0)
   &FN_InitializeAnimTrailBoneIndecies(AnimTrail Selection=1, First Bone or Socket Index=2, Second Bone or Socket Index=3) {enabled=false, module=FN_InitializeAnimTrailBoneIndecies001}
   &FN_InitializeAnimTrail(AnimTrailWidth=1.3, Lifetime=0.12)
   &AddVelocity(Velocity=<0 400 0>) {enabled=false}
  ]
  particleUpdateScript [
   &SolveForcesAndVelocity(Clamp Velocity=false, Limit Acceleration=false)
   &FN_UpdateAnimTrail() {module=FN_UpdateAnimTrail001}
   &ScaleRibbonWidth(Ribbon Width Scale=&FloatFromCurve(
     FloatCurve=DataInterface() {propertyValues {Curve {keys [
         {arriveTangent=5.7119, interpMode=RCIM_Cubic, leaveTangent=5.7119, tangentMode=RCTM_User, time=0, value=0.5}
         {arriveTangentWeight=0.0381, interpMode=RCIM_Cubic, leaveTangentWeight=0.3051, tangentMode=RCTM_User, time=0.1, value=1}
         {arriveTangentWeight=0.3051, interpMode=RCIM_Cubic, leaveTangentWeight=0.0381, tangentMode=RCTM_User, time=0.9, value=1}
         {arriveTangent=-5.7871, interpMode=RCIM_Cubic, leaveTangent=-5.7871, tangentMode=RCTM_User, time=1, value=0.5}
        ]}}}
     CurveIndex=&ReturnNormalizedExecIndex()
    )) {enabled=false}
   &ScaleColor(
    Scale Alpha=$System.SystemFade
    Scale RGB=&MakeVectorFromLinearColorRGB(Color=&MergeHelper_ColorSwitch(
      0=&MultiplyLinearColor(
       Linear Color=<rgba 0.997 1 0.9881 1>
       Scale Factor=&Vector4FromCurve(
        CurveIndex=Linked() {linkedVariable {name=NPC.FortniteNPC.TimeOfDay_0-1, type=NiagaraFloat}}
        Vector4Curve=DataInterface() {propertyValues {
          AlphaCurve {keys [<at 0 1>,<at 1 10>]}
          BlueCurve {keys [
            {
             arriveTangent=0.0005
             arriveTangentWeight=0.1844
             interpMode=RCIM_Cubic
             leaveTangent=0.0005
             leaveTangentWeight=0.1844
             tangentMode=RCTM_User
             tangentWeightMode=RCTWM_WeightedBoth
             time=0.0814
             value=0.6355
            }
            <at 0.0959 0.4595>
            <at 0.1104 0.4317>
            {
             arriveTangent=1.2853
             arriveTangentWeight=0.2985
             interpMode=RCIM_Cubic
             leaveTangent=1.2853
             leaveTangentWeight=0.2985
             tangentMode=RCTM_User
             tangentWeightMode=RCTWM_WeightedBoth
             time=0.1316
             value=0.5198
            }
            <at 0.2035 0.6014>
            <at 0.5057 0.9887>
            <at 0.6154 0.9558>
            <at 0.7483 0.4654>
            <at 0.8613 0.4212>
            <at 0.967 0.6355>
           ]}
          GreenCurve {keys [
            {
             arriveTangent=0.6573
             arriveTangentWeight=0.3221
             interpMode=RCIM_Cubic
             leaveTangent=0.6573
             leaveTangentWeight=0.3221
             tangentMode=RCTM_User
             tangentWeightMode=RCTWM_WeightedBoth
             time=0.0814
             value=0.5352
            }
            <at 0.0959 0.5408>
            <at 0.1104 0.4957>
            {
             arriveTangent=2.928
             arriveTangentWeight=0.2665
             interpMode=RCIM_Cubic
             leaveTangent=2.928
             leaveTangentWeight=0.2665
             tangentMode=RCTM_User
             tangentWeightMode=RCTWM_WeightedBoth
             time=0.1316
             value=0.6913
            }
            <at 0.2035 0.8286>
            <at 0.5057 0.9896>
            <at 0.6154 0.9578>
            <at 0.7483 0.645>
            <at 0.8613 0.6168>
            <at 0.967 0.5435>
           ]}
          RedCurve {keys [
            {arriveTangentWeight=0.3333, interpMode=RCIM_Cubic, tangentMode=RCTM_User, tangentWeightMode=RCTWM_WeightedBoth, time=0.0814, value=0.3729}
            <at 0.0959 0.7667>
            <at 0.1104 0.7477>
            {
             arriveTangent=1.3747
             arriveTangentWeight=0.2269
             interpMode=RCIM_Cubic
             leaveTangent=1.3747
             leaveTangentWeight=0.2269
             tangentMode=RCTM_User
             tangentWeightMode=RCTWM_WeightedBoth
             time=0.1316
             value=0.8525
            }
            <at 0.2035 0.976>
            <at 0.5057 1>
            <at 0.6154 0.9796>
            <at 0.7483 0.95>
            <at 0.8613 0.9473>
            <at 0.967 0.4444>
           ]}
         }}
       )
      )
      1=&MultiplyLinearColor(
       Linear Color=<rgba 0.6497 0.3732 0.1319 1>
       Scale Factor=&Vector4FromCurve(
        CurveIndex=Linked() {linkedVariable {name=NPC.FortniteNPC.TimeOfDay_0-1, type=NiagaraFloat}}
        Vector4Curve=DataInterface() {propertyValues {
          AlphaCurve {keys [<at 0.0866 1>,<at 1 1>]}
          BlueCurve {keys [<at 0.0311 0.711>,<at 0.0562 0.807>,<at 0.1289 0.4342>,<at 0.3233 0.5074>,<at 0.5374 1>,<at 0.9088 0.4212>,<at 1 0.7251>]}
          GreenCurve {keys [<at 0.0311 0.5896>,<at 0.0562 0.6795>,<at 0.1289 0.4969>,<at 0.3233 0.7297>,<at 0.5374 1>,<at 0.9088 0.6168>,<at 1 0.6474>]}
          RedCurve {keys [<at 0.0311 0.3933>,<at 0.0562 0.4735>,<at 0.1289 0.7454>,<at 0.3233 0.976>,<at 0.5374 1>,<at 0.9088 0.9473>,<at 1 0.3748>]}
         }}
       )
      )
      2=&MultiplyLinearColor(
       Linear Color=<rgba 0.2817 0.1482 0.0695 1>
       Scale Factor=&Vector4FromCurve(
        CurveIndex=Linked() {linkedVariable {name=NPC.FortniteNPC.TimeOfDay_0-1, type=NiagaraFloat}}
        Vector4Curve=DataInterface() {propertyValues {
          AlphaCurve {keys [<at 0 1>,<at 1 1>]}
          BlueCurve {keys [<at 0.0814 0.3525>,<at 0.1289 0.4342>,<at 0.3233 0.5074>,<at 0.5374 1>,<at 0.9088 0.4212>,<at 1 0.7251>]}
          GreenCurve {keys [<at 0.0814 0.3147>,<at 0.1289 0.4969>,<at 0.3233 0.7297>,<at 0.5374 1>,<at 0.9088 0.6168>,<at 1 0.6474>]}
          RedCurve {keys [<at 0.0814 0.1822>,<at 0.1289 0.7454>,<at 0.3233 0.976>,<at 0.5374 1>,<at 0.9088 0.9473>,<at 1 0.3748>]}
         }}
       )
      )
      3=&MultiplyLinearColor(
       Linear Color=<rgba 0.0827 0.0592 0.0978 3>
       Scale Factor=&Vector4FromCurve(
        CurveIndex=Linked() {linkedVariable {name=NPC.FortniteNPC.TimeOfDay_0-1, type=NiagaraFloat}}
        Vector4Curve=DataInterface() {propertyValues {
          AlphaCurve {keys [<at 0 1>,<at 1 1>]}
          BlueCurve {keys [
            {
             arriveTangent=-0.1803
             arriveTangentWeight=0.2479
             interpMode=RCIM_Cubic
             leaveTangent=-0.1803
             leaveTangentWeight=0.2479
             tangentMode=RCTM_User
             tangentWeightMode=RCTWM_WeightedBoth
             time=0.088
             value=0.8053
            }
            {
             arriveTangent=1.2853
             arriveTangentWeight=0.2985
             interpMode=RCIM_Cubic
             leaveTangent=1.2853
             leaveTangentWeight=0.2985
             tangentMode=RCTM_User
             tangentWeightMode=RCTWM_WeightedBoth
             time=0.1415
             value=0.5776
            }
            <at 0.2533 0.6014>
            <at 0.396 0.9784>
            <at 0.6665 0.9558>
            <at 0.8801 0.4212>
            {
             arriveTangent=0.0005
             arriveTangentWeight=0.1844
             interpMode=RCIM_Cubic
             leaveTangent=0.0005
             leaveTangentWeight=0.1844
             tangentMode=RCTM_User
             tangentWeightMode=RCTWM_WeightedBoth
             time=1
             value=0.807
            }
           ]}
          GreenCurve {keys [
            {
             arriveTangent=-0.1803
             arriveTangentWeight=0.2479
             interpMode=RCIM_Cubic
             leaveTangent=-0.1803
             leaveTangentWeight=0.2479
             tangentMode=RCTM_User
             tangentWeightMode=RCTWM_WeightedBoth
             time=0.088
             value=0.6814
            }
            {
             arriveTangent=2.928
             arriveTangentWeight=0.2665
             interpMode=RCIM_Cubic
             leaveTangent=2.928
             leaveTangentWeight=0.2665
             tangentMode=RCTM_User
             tangentWeightMode=RCTWM_WeightedBoth
             time=0.1415
             value=0.7682
            }
            <at 0.2533 0.8286>
            <at 0.396 0.9784>
            <at 0.6665 0.9578>
            <at 0.8801 0.6168>
            {
             arriveTangent=0.6573
             arriveTangentWeight=0.3221
             interpMode=RCIM_Cubic
             leaveTangent=0.6573
             leaveTangentWeight=0.3221
             tangentMode=RCTM_User
             tangentWeightMode=RCTWM_WeightedBoth
             time=1
             value=0.6795
            }
           ]}
          RedCurve {keys [
            {
             arriveTangent=-0.1803
             arriveTangentWeight=0.2479
             interpMode=RCIM_Cubic
             leaveTangent=-0.1803
             leaveTangentWeight=0.2479
             tangentMode=RCTM_User
             tangentWeightMode=RCTWM_WeightedBoth
             time=0.088
             value=0.4752
            }
            {
             arriveTangent=1.3747
             arriveTangentWeight=0.2269
             interpMode=RCIM_Cubic
             leaveTangent=1.3747
             leaveTangentWeight=0.2269
             tangentMode=RCTM_User
             tangentWeightMode=RCTWM_WeightedBoth
             time=0.1415
             value=0.9473
            }
            <at 0.2533 0.976>
            <at 0.396 0.9784>
            <at 0.6665 0.9796>
            <at 0.8801 0.9473>
            {arriveTangentWeight=0.3333, interpMode=RCIM_Cubic, tangentMode=RCTM_User, tangentWeightMode=RCTWM_WeightedBoth, time=1, value=0.4735}
           ]}
         }}
       )
      )
      Index=$User.ReactivityIndex
      Num=::NewEnumerator2
     ))
    ScaleA=true
    ScaleRGB=true
   )
   /SetVariables(Particles.SoftTrail_Color:LinearColor=&FNTimeOfDayColor(
     Night Color=&SelectLinearColorFromArray(
      Array Sampling Mode=::NewEnumerator1
      Direct Array Index=$User.ReactivityIndex
      Color Selection Array=DataInterface() {propertyValues {ColorData [<rgba 0.4125 0.552 0.6514 1>, <rgba 0.0933 0.1133 0.1402 1>, <rgba 0.0497 0.0506 0.0446 1>, <rgba 0.0064 0.0085 0.0412 1>]}}
      Random Seed=?
     )
     Morning Color=&SelectLinearColorFromArray(
      Array Sampling Mode=::NewEnumerator1
      Direct Array Index=$User.ReactivityIndex
      Color Selection Array=DataInterface() {propertyValues {ColorData [<rgba 0.8666 0.7163 0.6797 1>, <rgba 0.4327 0.2401 0.1046 1>, <rgba 0.1544 0.0707 0.0213 1>, <rgba 0.0695 0.0414 0.0685 3>]}}
      Random Seed=?
     )
     Day Color=&SelectLinearColorFromArray(
      Array Sampling Mode=::NewEnumerator1
      Direct Array Index=$User.ReactivityIndex
      Color Selection Array=DataInterface() {propertyValues {ColorData [<rgba 0.8666 0.8084 0.7942 1>, <rgba 0.4327 0.2826 0.177 1>, <rgba 0.1544 0.0707 0.0213 1>, <rgba 0.0471 0.0376 0.0648 3>]}}
      Random Seed=?
     )
     Evening Color=&SelectLinearColorFromArray(
      Array Sampling Mode=::NewEnumerator1
      Direct Array Index=$User.ReactivityIndex
      Color Selection Array=DataInterface() {propertyValues {ColorData [<rgba 0.7582 0.6523 0.5475 1>, <rgba 0.4327 0.2401 0.1046 1>, <rgba 0.1544 0.0707 0.0213 1>, <rgba 0.0365 0.0217 0.0359 3>]}}
      Random Seed=?
     )
    )) {isSetParameters=true}
  ]
  rendererClasses [&NiagaraRibbonRendererProperties]
  renderers [
   {rendererClass=&NiagaraRibbonRendererProperties}
   {rendererClass=&NiagaraRibbonRendererProperties, rendererIndex=1}
   {rendererClass=&NiagaraRibbonRendererProperties, rendererIndex=2}
  ]
 }}
summary {
 systemName=NS_Pickaxe_TopicProud_Trail
 emitters [{emitterName=FN_AnimTrail, rendererClasses [&NiagaraRibbonRendererProperties]}]
 userVariables [
  User.PowersActive {type=NiagaraBool}
  User.ReactivityIndex {type=NiagaraInt32}
  User.RibbonIDOffset {type=NiagaraInt32}
 ]
}
