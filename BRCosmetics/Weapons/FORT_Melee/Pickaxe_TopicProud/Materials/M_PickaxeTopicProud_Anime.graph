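# Anime-style cloth-model material for the TopicProud pickaxe, with lit and shaded colors that change with a ReactivityIndex parameter.
# The 2023 anime shader Pre, Calculations and Post functions form the toon pipeline, with an eye-animation function feeding lit and shaded colors.
# ReactivityIndex drives chained lerps through four lit and four shaded color stops.
# Named reroutes carry these ramps and the FX texture's red-channel mask toward color blends before Post.
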Material /BRCosmetics/Weapons/FORT_Melee/Pickaxe_TopicProud/Materials/M_PickaxeTopicProud_Anime
graph {
 properties {
  bAutomaticallySetUsageInEditor=false
  bUseMaterialAttributes=true
  bUsedWithClothing=true
  bUsedWithMeshDeformer=true
  bUsedWithMorphTargets=true
  bUsedWithNiagaraMeshParticles=true
  bUsedWithSkeletalMesh=true
  bUsedWithStaticLighting=true
  pixelDepthOffsetMode=PDOM_Legacy
  shadingModel=MSM_Cloth
  usedShadingModels=Cloth
 }
 nodes [
  n1 = MaterialFunctionCall!(CustomUV) {materialFunction=/Game/Global/Materials/Character_Parent_2023/Anime_Parent_2023/MaterialFunctions/MFC_AnimeShader_2023_Pre}
  n2 = MaterialFunctionCall!(CustomLightVector, OverrideNormal_WS, STT=@n1.STT, Normal_Map=@n1.STT, DFL=@n1.STT, FX=@n1.STT) {materialFunction=/Game/Global/Materials/Character_Parent_2023/Anime_Parent_2023/MaterialFunctions/MFC_AnimeShader_2023_Calculations}
  n3 = VertexColor()
  n4 = TextureSampleParameter2D() {
   group=01_EyeAnimation_Parameters
   parameterName=EyeMask
   samplerSource=SSM_Wrap_WorldGroupSettings
   samplerType=SAMPLERTYPE_Masks
   sortPriority=-32
   texture=&T_WhiteMask
  }
  n5 = StaticSwitchParameter(True=AppendVector(A=@n3, B=@n3), False=@n4.RGBA) {group=01_EyeAnimation_Parameters, parameterName=UseVertexColorForEyeMask, sortPriority=-33}
  n6 = TextureCoordinate() {coordinateIndex=1}
  n7 = Sign(None=Subtract(A=ComponentMask(None=@n6) {r=true}) {constB=0.0001})
  n8 = MaterialFunctionCall!(
   Color Lit=@n1.Color_Lit
   Color Shaded=@n1.Color_Lit
   Use UV2=StaticBoolParameter() {defaultValue=true, group=01_EyeAnimation_Parameters, parameterName=Use UV2 for Eye Animation, sortPriority=-30}
   Mask=StaticSwitchParameter(
    True=ChannelMaskParameter(None=@n5) {group=01_EyeAnimation_Parameters, parameterName=EyeMaskChannel, sortPriority=-32}.R
    False=Saturate(None=@n7)
   ) {group=01_EyeAnimation_Parameters, parameterName=UseEyeMask, sortPriority=-33}
  ) {materialFunction=/Game/Global/Materials/Character_Parent_2023/Anime_Parent_2023/MaterialFunctions/MF_AnimeEyeAnimation_v2}
  n9 = NamedRerouteUsage()
  MaterialFunctionCall!(
   Shading_Gradient=@n2.Shading_Gradient
   Threshold_Gradient=@n2.Shading_Gradient
   Specular_Gradient=@n2.Shading_Gradient
   Sheen_Mask=@n2.Shading_Gradient
   Thickness=@n2.Shading_Gradient
   DFL_Mask=@n2.Shading_Gradient
   FX=@n2.Shading_Gradient
   Metallic_Gradient=@n2.Shading_Gradient
   RimLight_Gradient=@n2.Shading_Gradient
   NormalBuffer=@n2.Shading_Gradient
   --------
   CINE_LightSources=@n2.Shading_Gradient
   CINE_LightColor=@n2.Shading_Gradient
   ---------
   Color_Lit=LinearInterpolate(A=@n8.Lit, B=NamedRerouteUsage(), Alpha=@n9)
   Color_Shaded=LinearInterpolate(A=@n8.Shaded, B=NamedRerouteUsage(), Alpha=@n9)
   SRM=@n1.SRM
   Emissive=@n1.SRM
   DFL_Color=@n1.SRM
   Specular_Color=@n1.SRM
   RimLight_Color=@n1.SRM
  ) {materialFunction=/Game/Global/Materials/Character_Parent_2023/Anime_Parent_2023/MaterialFunctions/MFC_AnimeShader_2023_Post}
  n10 = TextureSampleParameter2D() {
   group=Reactivity
   parameterName=TechArtMask_RE
   samplerType=SAMPLERTYPE_Masks
   texture=/BRCosmetics/Weapons/FORT_Melee/Pickaxe_TopicProud/Textures/T_TopicProud_Pickaxe_FX
  }
  NamedRerouteDeclaration(None=@n10.R) {name=CustomArea, nodeColor=<rgba 1 0.9609 0 1>}
  n11 = Saturate(None=Divide(A=ScalarParameter() {group=Reactivity, parameterName=ReactivityIndex}) {constB=3})
  n12 = Multiply(A=@n11) {constB=3}
  n13 = LinearInterpolate(
   A=VectorParameter() {group=Reactivity, parameterName=Lit01, defaultValue=<rgba 1 1 0.9911 1>}.RGB
   B=VectorParameter() {group=Reactivity, parameterName=Lit02, defaultValue=<rgba 0.807 0.5457 0.2705 1>}.RGB
   Alpha=Saturate(None=@n12)
  )
  n14 = VectorParameter() {group=Reactivity, parameterName=Lit03, defaultValue=<rgba 0.3419 0.1651 0.063 1>}
  n15 = Subtract(A=@n12)
  n16 = Saturate(None=@n15)
  n17 = Subtract(A=@n15)
  NamedRerouteDeclaration(None=LinearInterpolate(
    A=LinearInterpolate(A=@n13, B=@n14.RGB, Alpha=@n16)
    B=VectorParameter() {group=Reactivity, parameterName=Lit04, defaultValue=<rgba 0.0865 0.0648 0.0999 1>}.RGB
    Alpha=Saturate(None=@n17)
   )) {name=LitColorComp, nodeColor=<rgba 1 0.3984 0 1>}
  n18 = Multiply(A=@n11) {constB=3}
  n19 = LinearInterpolate(
   A=VectorParameter() {group=Reactivity, parameterName=Shaded01, defaultValue=<rgba 0.4233 0.4004 0.3045 1>}.RGB
   B=VectorParameter() {group=Reactivity, parameterName=Shaded02, defaultValue=<rgba 0.3712 0.227 0.0793 1>}.RGB
   Alpha=Saturate(None=@n18)
  )
  n20 = VectorParameter() {group=Reactivity, parameterName=Shaded03, defaultValue=<rgba 0.124 0.0622 0.0176 1>}
  n21 = Subtract(A=@n18)
  n22 = Saturate(None=@n21)
  n23 = Subtract(A=@n21)
  NamedRerouteDeclaration(None=LinearInterpolate(
    A=LinearInterpolate(A=@n19, B=@n20.RGB, Alpha=@n22)
    B=VectorParameter() {group=Reactivity, parameterName=Shaded04, defaultValue=<rgba 0.0395 0.0307 0.0343 1>}.RGB
    Alpha=Saturate(None=@n23)
   )) {name=ShadedColorComp, nodeColor=<rgba 0 0.4844 1 1>}
  TemporalResponsivenessOutput!(None)
 ]
}
