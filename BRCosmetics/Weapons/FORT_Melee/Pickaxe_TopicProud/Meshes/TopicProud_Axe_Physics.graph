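# Physics asset for the TopicProud pickaxe mesh, a Fortnite melee weapon cosmetic.
# It has two bodies, a thin capsule on stick_01 and a sphere on stick_02.
# A single constraint joins stick_02 to stick_01 with narrow swing limits and a zero twist limit, so the second segment can only sway slightly relative to the first.
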PhysicsAsset /BRCosmetics/Weapons/FORT_Melee/Pickaxe_TopicProud/Meshes/TopicProud_Axe_Physics
bodies {
 stick_01 [{shapeType=Capsule, length=24.5, radius=2, center=<12.0789 1.1655 0.1803>, rotation=<rot 0 101.9272 -88.8536>}]
 stick_02 [{shapeType=Sphere, radius=7.9944, shapeName=stick_02_sphere, center=<4.7399 0 0>}]
}
constraints [{bone1Name=stick_02, bone2Name=stick_01, swing1LimitDegrees=10, swing2LimitDegrees=10, twistLimitDegrees=0}]
