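# Animation Blueprint for the TopicProud pickaxe that adds physics sway to its stick bones outside the front end.
# The cached input pose runs through two chained spring-driven AnimDynamics nodes on stick_01 and stick_02, while the front end uses the unmodified pose.
# A RigidBody blend switched by FE_withpawn, updated each frame from front-end and on-pawn state, is present but not connected to the output.
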AnimBlueprint /BRCosmetics/Weapons/FORT_Melee/Pickaxe_TopicProud/Meshes/TopicProud_Axe_AnimBP
target_skeleton=&/TopicProud_Axe_Skeleton
graphs [
 AnimGraph {nodes [
   n1 = RigidBody(ComponentPose=Local To Component(LocalPose=Use cached pose 'for FE'()), Alpha=0.25) {
    overridePhysicsAsset=&/TopicProud_Axe_Physics
    simulationSpace=WorldSpace
    baseBoneRef {boneName=b_AR_Root}
    componentAppliedLinearAccClamp=<3000 3000 3000>
    componentLinearAccScale=<0.1 0.1 0.1>
    componentLinearVelScale=<0.1 0.1 0.1>
    overrideWorldGravity {z=-980}
    simSpaceSettings {dampingAlpha=0, maxAngularAcceleration=15, maxAngularVelocity=3, maxLinearAcceleration=1500, maxLinearVelocity=500, worldAlpha=1}
   }
   n2 = RigidBody(ComponentPose=Local To Component(LocalPose=Use cached pose 'for FE'()), Alpha=0.23) {
    overridePhysicsAsset=&TopicProud_Axe_Physics
    simulationSpace=WorldSpace
    baseBoneRef {boneName=Melee_TwoHanded_Handle}
    componentAppliedLinearAccClamp=<3000 3000 3000>
    overrideWorldGravity {z=-980}
    simSpaceSettings {dampingAlpha=0, maxAngularAcceleration=15, maxAngularVelocity=5, maxLinearAcceleration=2100, maxLinearVelocity=750, worldAlpha=1}
   }
   Blend Poses by bool(
    BlendPose_0=Component To Local(ComponentPose=@n1)
    BlendPose_1=Component To Local(ComponentPose=@n2)
    bActiveValue=Get FE_withpawn() -> (FE_withpawn:bool)
   )
   n3 = Anim Dynamics - Bone: stick_01(ComponentPose=Local To Component(LocalPose=Use cached pose 'for FE'()), Alpha=1, ExternalForce="0,0,0") {
    angularDampingOverride=0.9999
    angularSpringConstant=2000
    bAngularSpring=true
    bOverrideAngularDamping=true
    bUseGravityOverride=true
    simulationSpace=Actor
    boundBone {boneName=stick_01}
    componentAppliedLinearAccClamp=<4000 4000 6000>
    componentLinearAccScale=<2 2 0.5>
    componentLinearVelScale=<0.1 0.1 0.5>
    gravityOverride {z=-980}
    physicsBodyDefinitions [{
      boundBone {boneName=stick_01}
      boxExtents=<10 10 10>
      constraintSetup {angularLimitsMax=<0 5 2>, angularLimitsMin=<0 -5 -2>, angularTarget=<1 0 0>}
      localJointOffset=<0.5 0.3 0>
     }]
    relativeSpaceBone {boneName=Melee_TwoHanded_Handle}
   }
   n4 = Anim Dynamics - Bone: stick_02(ComponentPose=@n3, Alpha=1, ExternalForce="0,0,0") {
    angularDampingOverride=0.9999
    angularSpringConstant=1500
    bAngularSpring=true
    bOverrideAngularDamping=true
    bUseGravityOverride=true
    simulationSpace=BoneRelative
    boundBone {boneName=stick_02}
    componentAppliedLinearAccClamp=<4000 4000 6000>
    componentLinearAccScale=<2 2 0.5>
    componentLinearVelScale=<0.1 0.1 0.5>
    gravityOverride {z=-980}
    physicsBodyDefinitions [{
      boundBone {boneName=stick_02}
      boxExtents=<10 10 10>
      constraintSetup {angularLimitsMax=<0 8 8>, angularLimitsMin=<0 -8 -8>, angularTarget=<0 0 0>}
      localJointOffset=<0.5 0.3 0>
     }]
    relativeSpaceBone {boneName=Melee_TwoHanded_Handle}
   }
   Output Pose(Result=Blend Poses by bool(BlendPose_0=Use cached pose 'for FE'(), BlendPose_1=Component To Local(ComponentPose=@n4), bActiveValue=Get bIsInFrontEnd())) {layerGroup=None}
   AnimGraphNode_SubInput_0 = Input Pose()
   Save cached pose 'for FE'(Pose=@AnimGraphNode_SubInput_0) {cacheName=for FE}
  ]}
 EventGraph {nodes [
   Event BlueprintUpdateAnimation()
   > Set FE_withpawn(FE_withpawn:bool=AND Boolean(A=Get bIsInFrontEnd(), B=Get bIsOnPawn() -> (bIsOnPawn:bool))) -> (Output_Get:bool)
   TryGetPawnOwner()
  ]}
]
variables [FE_withpawn {type=boolean}]
